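; Front-end diorama actor that previews a toy cosmetic through a child actor component.
; On the preview toy event, it async loads the toy definition's actor class and sets it as the child actor's class.
; It then sends the child a front-end preview initialization message, hides it and spawns a character transition particle at its location.
; After a brief delay it reveals the toy, so the effect covers the swap.
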(event Spray|EventPreviewToy
  (params (Toy :type Athena%20Toy%20Item%20Definition%20Object%20Reference))
  (bind _toyspawncac (Variables|Default|GetToySpawnCAC))
  (bind _childactor (Variables|/Script/Engine.ChildActorComponent|GetChildActor _toyspawncac))
  (bind _class (Utilities|AsyncLoadClassAsset (Class|AthenaToyItemDefinition|GetToyActorClass Toy))
    (:Completed
      (bind _asactor (Utilities|Casting|CastToActorClass _class))
      (ChildActorComponent|SetChildActorClass _toyspawncac _asactor)
      (Visuals|InitializeToyPreviewinFrontend(Message) _childactor)
      (Rendering|SetActorHiddenInGame _childactor true)
      (Effects|SpawnEmitteratLocation "/Game/Effects/Fort_Effects/UI/Transition/P_FrontEndCharacterTransition.P_FrontEndCharacterTransition" (Transformation|GetActorLocation _childactor) "0,90.000000, 0")
      (Utilities|FlowControl|Delay 0.1)
      (Rendering|SetActorHiddenInGame _childactor false))))

(fn ConstructionScript)
